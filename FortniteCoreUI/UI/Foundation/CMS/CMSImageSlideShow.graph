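# UMG widget that shows a slideshow of CMS images in an auto-scrolling CommonUI carousel.
# SetMediaURLs starts auto-scrolling at the ScrollInterview interval, clears the carousel, then creates and adds one CMSImage child per URL and passes it that URL.
# Carousel page changes are re-broadcast through the CarouselIndexChanged dispatcher with the current page index.
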WidgetBlueprint /FortniteCoreUI/UI/Foundation/CMS/CMSImageSlideShow
graphs [
 CarouselIndexChanged {nodes [CarouselIndexChanged!() -> (then, CurrentPageIndex:int) {is=title K2Node_FunctionEntry}]}
 EventGraph {nodes [
   n1 = "On Current Page Index Changed (Images)"!() -> (OutputDelegate, then, CarouselWidget:Common Widget Carousel Object Reference, CurrentPageIndex:int=0) {is=title K2Node_ComponentBoundEvent}
   > Call Carousel Index Changed!(execute:Exec, self:CMSImage Slide Show Object Reference, CurrentPageIndex:int=@n1.CurrentPageIndex) -> (then:Exec) {is=title K2Node_CallDelegate}
   n2 = SetMediaURLs!() -> (OutputDelegate, then, MediaUrls:Array of Strings) {is=title K2Node_CustomEvent}
   > BeginAutoScrolling!(
    execute
    self:Common Widget Carousel Object Reference=Get Images() -> (Images:Common Widget Carousel Object Reference)
    ScrollInterval:float=Get ScrollInterview() -> (ScrollInterview:double)
   ) -> (then)
   > ClearChildren(self=Get Images())
   > n3 = For Each Loop(Array=@n2.MediaUrls)
   > n4 = Create Widget!(execute:Exec, Class:User Widget Class Reference=/Game/UI/Foundation/CMS/CMSImage.CMSImage_C, OwningPlayer:Player Controller Object Reference) -> (then:Exec, ReturnValue:CMSImage Object Reference)
   > AddChild(self=Get Images(), Content=@n4)
   > SetMediaURL!(execute, self:CMSImage Object Reference=@n4, MediaUrl:String=@n3) -> (then)
  ]}
]
variables [
 CarouselIndexChanged {}
 ScrollInterview {type=number, default=3}
]
widgets {
 info {inheritedWidgetCount=0, parentClass=&UserWidget, rootWidgetClass=/Script/CommonUI.CommonWidgetCarousel, widgetCount=1}
 widgets [Images {widgetClassPath=&CommonWidgetCarousel, bIsVariable=true, widget=Images}]
}
